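; Gameplay ability for the Clyde wig item that clears the wig effect and its attached disguise actor from the player.
; On activation it sets up the generic ability, then force-removes the item if none are left in inventory.
; It removes effects granting the Clyde.Item.Wig tag and forces a net update on the pawn.
; RemoveBush then destroys any attached Clyde_Player_Wig actors and ends the ability.
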(event Ability|EventActivateAbility
  (bind _playerpawn (Variables|Default|GetPlayerPawn))
  (bind _wigitemdef (Variables|Default|GetWigItemDef))
  (|CallParentFunction)
  (bind (_event_data _fort_player_pawn) (Setup|SetupGenericAbility false false)
    (:Output
      (Variables|Default|SetPlayerPawn _fort_player_pawn)
      ;[1 1 1] Remove Item from player if they don't have any charges left
      (bind _asfort_player_controller (Utilities|Casting|CastToFortPlayerController (Pawn|GetController (Variables|Default|GetPlayerPawn)))
        (:then
          (Utilities|IsValid _wigitemdef
            (:Is%20Valid
              (if (== (Inventory|GetItemQuantityOnPlayer _asfort_player_controller _wigitemdef) 0)
                (Inventory|RemoveItemFromPlayer :PlayerController _asfort_player_controller :ItemDefinition (Variables|Default|GetWigItemDef) :AmountToRemove -1 :bForceRemoval true :bSkipVerbRemoval false)
                (else
                  _))
              (Ability|RemoveGameplayEffectFromOwnerWithGrantedTags "(GameplayTags=((TagName=\"Clyde.Item.Wig\")))")
              (Networking|ForceNetUpdate _playerpawn)
              ;[1 1 1] Remove the Bush in case they took damage
              (CallFunction|RemoveBush _playerpawn))
            (:Is%20Not%20Valid
              _)))
        (:CastFailed
          _)))))

(fn RemoveBush
  (params (PlayerPawn :type Fort%20Player%20Pawn%20Object%20Reference))
  (for _array_element (Actor|GetAttachedActors PlayerPawn true false)
    (bind _asclyde_player_wig (Utilities|Casting|CastToClyde_Player_Wig _array_element))
    (Utilities|IsValid _asclyde_player_wig
      (:Is%20Valid
        (Actor|DestroyActor _asclyde_player_wig))))
  (Ability|EndAbility))
